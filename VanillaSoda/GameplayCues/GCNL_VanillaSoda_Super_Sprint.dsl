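; Looping gameplay cue for the VanillaSoda super sprint that drives its audio and particle effects on the player pawn.
; On loop start, after a zero-length delay, it fades in an attached force loop sound and passes local-control state to the first particle component.
; It also binds the VanillaSoda weapon mesh to the second component's Niagara "SwordMesh" input.
; On removal, it resends the local flag and fades out the stored loop.
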(event EventOnLoopingStart(EitherVFX-preferthisversion)
  (params (MyTarget :type Actor%20Object%20Reference) (Parameters :type Gameplay%20Cue%20Parameters%20Structure%20%28by%20ref%29) (ParticleComponents :type Array%20of%20FXSystem%20Component%20Object%20References) (AudioComponents :type Array%20of%20Audio%20Component%20Object%20References))
  (Utilities|FlowControl|Delay 0)
  (bind _asfort_player_pawn_athena (Utilities|Casting|CastToFortPlayerPawnAthena MyTarget))
  ;[1 1 1] Audio - Fade In Loop
  (bind _returnvalue (Audio|SpawnSoundAttached (Variables|Default|GetForceLoop) (Variables|/Script/Engine.Actor|GetRootComponent _asfort_player_pawn_athena)))
  (bind _output_get (Variables|Default|SetForceLoopComponent _returnvalue))
  (Audio|Components|Audio|FadeIn _output_get 2.0 1.0 0.0 "Linear" :!class /Script/Engine.AudioComponent)
  (Variables|Default|SetIsLocal (Pawn|IsLocallyControlled _asfort_player_pawn_athena))
  (Effects|Components|ParticleSystem|SetBoolParameter (Utilities|Array|Get(acopy) ParticleComponents) "bIsLocalClient" (Variables|Default|GetIsLocal))
  (bind _asniagara_particle_system_component (Utilities|Casting|CastToNiagaraComponent (Utilities|Array|Get(acopy) ParticleComponents 1)))
  (bind _asb_vanilla_soda_weapon (Utilities|Casting|CastToB_VanillaSoda_Weapon (Variables|/Script/FortniteGame.FortPawn|GetCurrentWeapon _asfort_player_pawn_athena)))
  (Niagara|SetNiagaraSkeletalMeshComponent _asniagara_particle_system_component "SwordMesh" (Variables|/Script/FortniteGame.FortWeapon|GetWeaponMesh _asb_vanilla_soda_weapon)))

(event EventOnRemoval(EitherVFX-preferthisversion)
  (params (MyTarget :type Actor%20Object%20Reference) (Parameters :type Gameplay%20Cue%20Parameters%20Structure%20%28by%20ref%29) (ParticleComponents :type Array%20of%20FXSystem%20Component%20Object%20References) (AudioComponents :type Array%20of%20Audio%20Component%20Object%20References) (BurstCameraShakeInstance :type Camera%20Shake%20Base%20Object%20Reference) (BurstDecalInstance :type Decal%20Actor%20Object%20Reference))
  (Effects|Components|ParticleSystem|SetBoolParameter (Utilities|Array|Get(acopy) ParticleComponents) "bIsLocalClient" (Variables|Default|GetIsLocal))
  (Utilities|Casting|CastToFortPlayerPawnAthena MyTarget)
  ;[1 1 1] Audio - Fade Out Loop
  (bind _forceloopcomponent (Variables|Default|GetForceLoopComponent)
    (:then
      (Audio|Components|Audio|FadeOut _forceloopcomponent 1.0 0.0 "Linear" :!class /Script/Engine.AudioComponent))
    (:else)))

(fn ConstructionScript)
