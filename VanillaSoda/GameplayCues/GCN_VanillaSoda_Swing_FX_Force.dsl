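; A burst gameplay cue for the VanillaSoda swing that spawns a radial physics force in front of the swinging player.
; On burst, it spawns an FN_RadialForce field actor offset horizontally along the target pawn's capsule forward vector.
; The spawn is cosmetic-gated: it runs only off mobile, outside PC performance mode and at a rendering detail mode above 1.
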(fn OnBurst(EitherVFX-preferthisversion)
  (params (MyTarget :type Actor%20Object%20Reference) (Parameters :type Gameplay%20Cue%20Parameters%20Structure%20%28by%20ref%29) (ParticleComponents :type Array%20of%20FXSystem%20Component%20Object%20References) (AudioComponents :type Array%20of%20Audio%20Component%20Object%20References) (BurstCameraShakeInstance :type Camera%20Shake%20Base%20Object%20Reference) (BurstDecalInstance :type Decal%20Actor%20Object%20Reference))
  (if (Math|Boolean|ANDBoolean (not (Fortnite|IsMobilePlatform)) (not (Class|PlatformHelpersLibrary|IsPcPerfMode 0)) (> (Rendering|GetRenderingDetailMode) 1))
    (Game|SpawnActorfromClass 0 "/Game/Effects/Niagara/Blueprints/PhsyicsFields/FN_RadialForce.FN_RadialForce_C" (Math|Conversions|ToTransform(Vector) (+ (* (Transformation|GetForwardVector (Variables|/Script/Engine.Character|GetCapsuleComponent (Utilities|Casting|CastToFortPlayerPawnAthena MyTarget))) "150.000000,150.000000,0.000000") (Transformation|GetWorldLocation (Variables|/Script/Engine.Character|GetCapsuleComponent (Utilities|Casting|CastToFortPlayerPawnAthena MyTarget))))) "Undefined" "MultiplyWithRoot" 0 2.0 0.5 7e+02 1.0 0.05 0)
    (return)))
